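# Physics asset for the collar of the male medium Bone Wasp player body, set up for secondary motion of its dyn_collar bones.
# Box bodies form seven collar chains, each constrained to a tiny capsule on its MSTR bone, with identical swing and twist limits.
# The two front chains have two segments each, and most right-side boxes are larger than their left counterparts.
# An unconstrained head capsule provides a collision shape.
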PhysicsAsset /BRCosmetics/Characters/Player/Male/Medium/Bodies/M_MED_Bone_Wasp/Meshes/M_MED_Bone_Wasp_Dyn_Collar
bodies {
 dyn_collar_back [{shapeType=Box, extentX=8.75, extentY=18.5, extentZ=1.1, center=<4.8371 0 0.3247>}]
 dyn_collar_back_MSTR [{shapeType=Capsule, length=0.5, radius=0.5, rotation=<rot 90 0 0>}]
 dyn_collar_bk_l [{shapeType=Box, extentX=8.4375, extentY=18.5, extentZ=1.1, center=<4.3416 0 0.0001>}]
 dyn_collar_bk_l_MSTR [{shapeType=Capsule, length=0.5, radius=0.5, rotation=<rot 90 0 0>}]
 dyn_collar_bk_r [{shapeType=Box, extentX=7.3125, extentY=21.1875, extentZ=2.1, center=<5.5604 0 0.4209>}]
 dyn_collar_bk_r_MSTR [{shapeType=Capsule, length=0.5, radius=0.5, rotation=<rot 90 0 0>}]
 dyn_collar_fr_l_01 [{shapeType=Box, extentX=6, extentY=10.5, extentZ=1.1, center=<2.4084 0 -0.0001>}]
 dyn_collar_fr_l_02 [{shapeType=Box, extentX=6, extentY=10.5, extentZ=1.1, center=<2.4084 0 -0.0001>}]
 dyn_collar_fr_l_MSTR [{shapeType=Capsule, length=0.5, radius=0.5, rotation=<rot 90 0 0>}]
 dyn_collar_fr_r_01 [{shapeType=Box, extentX=6, extentY=18.1875, extentZ=2.725, center=<2.4084 0 -0.0001>}]
 dyn_collar_fr_r_02 [{shapeType=Box, extentX=6, extentY=18.1875, extentZ=2.725, center=<2.4084 0 -0.0001>}]
 dyn_collar_fr_r_MSTR [{shapeType=Capsule, length=0.5, radius=0.5, rotation=<rot 90 0 0>}]
 dyn_collar_l [{shapeType=Box, extentX=7.375, extentY=18.5, extentZ=1.1, center=<5.6641 0 0.0001>}]
 dyn_collar_l_MSTR [{shapeType=Capsule, length=0.5, radius=0.5, rotation=<rot 90 0 0>}]
 dyn_collar_r [{shapeType=Box, extentX=7.625, extentY=23.6875, extentZ=2.475, center=<6.5574 0.0001 -0.8599>}]
 dyn_collar_r_MSTR [{shapeType=Capsule, length=0.5, radius=0.5, rotation=<rot 90 0 0>}]
 head [{shapeType=Capsule, length=2.9375, radius=13.2153, center=<2.2493 0.0002 0>, rotation=<rot 90 0 0>}]
}
constraints [
 {bone1Name=dyn_collar_back, bone2Name=dyn_collar_back_MSTR, swing1LimitDegrees=75, swing2LimitDegrees=75, twistLimitDegrees=90}
 {bone1Name=dyn_collar_bk_l, bone2Name=dyn_collar_bk_l_MSTR, swing1LimitDegrees=75, swing2LimitDegrees=75, twistLimitDegrees=90}
 {bone1Name=dyn_collar_l, bone2Name=dyn_collar_l_MSTR, swing1LimitDegrees=75, swing2LimitDegrees=75, twistLimitDegrees=90}
 {bone1Name=dyn_collar_fr_l_01, bone2Name=dyn_collar_fr_l_MSTR, swing1LimitDegrees=75, swing2LimitDegrees=75, twistLimitDegrees=90}
 {bone1Name=dyn_collar_fr_l_02, bone2Name=dyn_collar_fr_l_01, swing1LimitDegrees=75, swing2LimitDegrees=75, twistLimitDegrees=90}
 {bone1Name=dyn_collar_bk_r, bone2Name=dyn_collar_bk_r_MSTR, swing1LimitDegrees=75, swing2LimitDegrees=75, twistLimitDegrees=90}
 {bone1Name=dyn_collar_r, bone2Name=dyn_collar_r_MSTR, swing1LimitDegrees=75, swing2LimitDegrees=75, twistLimitDegrees=90}
 {bone1Name=dyn_collar_fr_r_01, bone2Name=dyn_collar_fr_r_MSTR, swing1LimitDegrees=75, swing2LimitDegrees=75, twistLimitDegrees=90}
 {bone1Name=dyn_collar_fr_r_02, bone2Name=dyn_collar_fr_r_01, swing1LimitDegrees=75, swing2LimitDegrees=75, twistLimitDegrees=90}
]
