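# Physics asset for the Bone Wasp medium male player body's dynamic jacket ("no_back" variant, though back bodies remain), under-jacket flaps and left thigh pouch.
# Two-segment box chains hang from spine_02, mostly with limited swing and twist, over pelvis, spine and thigh capsules.
# Free-motion constraints cross-link lower side panels to back panels, plus front to side on the left only.
# Under-jacket flaps pivot from small master capsules linked to spine_01.
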PhysicsAsset /BRCosmetics/Characters/Player/Male/Medium/Bodies/M_MED_Bone_Wasp/Meshes/M_MED_Bone_Wasp_Dyn_Jacket_no_back
bodies {
 dyn_jacket_bk_01 [{shapeType=Box, extentX=4.5625, extentY=1, extentZ=1.6, center=<6.3696 0 -0.0001>}]
 dyn_jacket_bk_02 [{shapeType=Box, extentX=13, extentY=14, extentZ=1.6, center=<6.3696 0 -0.0001>}]
 dyn_jacket_bk_l_01 [{shapeType=Box, extentX=13, extentY=14, extentZ=1.6, center=<6.3696 0 -0.0001>}]
 dyn_jacket_bk_l_02 [{shapeType=Box, extentX=13, extentY=14, extentZ=2.1, center=<6.3696 0 -0.0001>}]
 dyn_jacket_bk_r_01 [{shapeType=Box, extentX=13, extentY=14, extentZ=1.6, center=<6.3696 0 -0.0001>}]
 dyn_jacket_bk_r_02 [{shapeType=Box, extentX=13, extentY=19.4375, extentZ=2.1, center=<6.3696 0 -0.0001>}]
 dyn_jacket_fr_l_01 [{shapeType=Box, extentX=18.1875, extentY=14, extentZ=1.1, center=<10.665 -0.0001 -0.0001>}]
 dyn_jacket_fr_l_02 [{shapeType=Box, extentX=17.375, extentY=14, extentZ=1.1, center=<9.4187 0 0>, rotation=<rot 0 0 20.001>}]
 dyn_jacket_fr_r_01 [{shapeType=Box, extentX=18.1875, extentY=14, extentZ=1.1, center=<9.2034 0 -0.0001>}]
 dyn_jacket_fr_r_02 [{shapeType=Box, extentX=18.1875, extentY=14, extentZ=1.1, center=<9.394 0 -0.0001>}]
 dyn_jacket_l_01 [{shapeType=Box, extentX=13, extentY=14, extentZ=2.1, center=<6.3696 0 -0.0001>}]
 dyn_jacket_l_02 [{shapeType=Box, extentX=17.75, extentY=14, extentZ=2.1, center=<9.6786 -0.0003 -0.0001>}]
 dyn_jacket_r_01 [{shapeType=Box, extentX=13, extentY=14, extentZ=2.1, center=<6.3696 0 -0.0001>}]
 dyn_jacket_r_02 [{shapeType=Box, extentX=13, extentY=18.5, extentZ=2.1, center=<6.3696 0 -0.0001>}]
 dyn_thigh_pouch_l [{shapeType=Capsule, length=4.516, radius=8.3971, center=<2.5945 -0.5333 -1.782>, rotation=<rot 84.9999 -44.9985 0>}]
 dyn_under_jacket_fr_l_01 [{shapeType=Box, extentX=4.25, extentY=8, extentZ=1.1, center=<7.3417 0.1159 0.3>, rotation=<rot 0 0 13.177>}]
 dyn_under_jacket_fr_l_MSTR [{shapeType=Capsule, length=0.505, radius=0.505, rotation=<rot 0 -90 0>}]
 dyn_under_jacket_fr_r_01 [{shapeType=Box, extentX=7.5, extentY=8, extentZ=1.1, center=<5.8577 -0.2661 1.2>, rotation=<rot 0 0 -12.2025>}]
 dyn_under_jacket_fr_r_MSTR [{shapeType=Capsule, length=0.505, radius=0.505, rotation=<rot 0 -90 0>}]
 pelvis [
  {shapeType=Capsule, length=9.8125, radius=12.6875, center=<0 -0.9679 0>}
  {shapeType=Capsule, length=9.4688, radius=9, center=<-6.6976 3.9057 0>}
 ]
 spine_01 [{shapeType=Capsule, length=7.8125, radius=14.0625, center=<1.3298 -0.0003 0>}]
 spine_02 [{shapeType=Capsule, length=5.75, radius=13.5, center=<-4.3551 -0.0001 0>}]
 thigh_l [{shapeType=Capsule, length=38.2729, radius=9.1765, center=<-23.807 -0.7638 0.0002>, rotation=<rot 90 3.5763 -1.0255>}]
 thigh_r [{shapeType=Capsule, length=22.2104, radius=9.1765, center=<15.7286 0.0006 -0.0002>, rotation=<rot 90 0 0>}]
}
constraints [
 {bone1Name=dyn_jacket_fr_l_02, bone2Name=dyn_jacket_fr_l_01, swing1LimitDegrees=2, swing2LimitDegrees=50, twistLimitDegrees=8}
 {bone1Name=dyn_jacket_l_02, bone2Name=dyn_jacket_l_01, swing1LimitDegrees=15, swing2LimitDegrees=50, twistLimitDegrees=8}
 {bone1Name=dyn_jacket_bk_l_02, bone2Name=dyn_jacket_bk_l_01, swing1LimitDegrees=5, swing2LimitDegrees=5, twistLimitDegrees=5}
 {bone1Name=dyn_jacket_bk_r_02, bone2Name=dyn_jacket_bk_r_01, swing1LimitDegrees=10, swing2LimitDegrees=10, twistLimitDegrees=5}
 {bone1Name=dyn_jacket_r_02, bone2Name=dyn_jacket_r_01, swing1LimitDegrees=15, swing2LimitDegrees=50, twistLimitDegrees=8}
 {bone1Name=dyn_jacket_fr_r_02, bone2Name=dyn_jacket_fr_r_01, swing1LimitDegrees=5, swing2LimitDegrees=50, twistLimitDegrees=8}
 {bone1Name=dyn_jacket_fr_l_01, bone2Name=spine_02, swing1LimitDegrees=2, swing2LimitDegrees=75, twistLimitDegrees=50}
 {bone1Name=dyn_jacket_l_01, bone2Name=spine_02, swing1LimitDegrees=30, swing2LimitDegrees=50, twistLimitDegrees=8}
 {bone1Name=dyn_jacket_bk_l_01, bone2Name=spine_02, swing1LimitDegrees=5, swing2LimitDegrees=5, twistLimitDegrees=5}
 {bone1Name=dyn_jacket_bk_r_01, bone2Name=spine_02, swing1LimitDegrees=5, swing2LimitDegrees=5, twistLimitDegrees=5}
 {bone1Name=dyn_jacket_r_01, bone2Name=spine_02, swing1LimitDegrees=30, swing2LimitDegrees=50, twistLimitDegrees=8}
 {bone1Name=dyn_jacket_fr_r_01, bone2Name=spine_02, swing1LimitDegrees=5, swing2LimitDegrees=50, twistLimitDegrees=2}
 {bone1Name=dyn_under_jacket_fr_l_01, bone2Name=dyn_under_jacket_fr_l_MSTR, swing1LimitDegrees=75, swing2LimitDegrees=75, twistLimitDegrees=25}
 {bone1Name=dyn_under_jacket_fr_r_01, bone2Name=dyn_under_jacket_fr_r_MSTR, swing1LimitDegrees=45, swing2LimitDegrees=45, twistLimitDegrees=5}
 {
  bone1Name=spine_01
  bone2Name=pelvis
  swing1LimitDegrees=45
  swing1Motion=Free
  swing2LimitDegrees=45
  swing2Motion=Free
  twistLimitDegrees=45
  twistMotion=Free
 }
 {
  bone1Name=spine_02
  bone2Name=spine_01
  swing1LimitDegrees=45
  swing1Motion=Free
  swing2LimitDegrees=45
  swing2Motion=Free
  twistLimitDegrees=45
  twistMotion=Free
 }
 {
  bone1Name=dyn_jacket_r_02
  bone2Name=dyn_jacket_bk_r_02
  swing1LimitDegrees=45
  swing1Motion=Free
  swing2LimitDegrees=45
  swing2Motion=Free
  twistLimitDegrees=45
  twistMotion=Free
 }
 {
  bone1Name=dyn_jacket_l_02
  bone2Name=dyn_jacket_bk_l_02
  swing1LimitDegrees=45
  swing1Motion=Free
  swing2LimitDegrees=45
  swing2Motion=Free
  twistLimitDegrees=45
  twistMotion=Free
 }
 {
  bone1Name=thigh_l
  bone2Name=pelvis
  swing1LimitDegrees=45
  swing1Motion=Free
  swing2LimitDegrees=45
  swing2Motion=Free
  twistLimitDegrees=45
  twistMotion=Free
 }
 {
  bone1Name=thigh_r
  bone2Name=pelvis
  swing1LimitDegrees=45
  swing1Motion=Free
  swing2LimitDegrees=45
  swing2Motion=Free
  twistLimitDegrees=45
  twistMotion=Free
 }
 {
  bone1Name=dyn_jacket_fr_l_02
  bone2Name=dyn_jacket_l_02
  swing1LimitDegrees=45
  swing1Motion=Free
  swing2LimitDegrees=45
  swing2Motion=Free
  twistLimitDegrees=45
  twistMotion=Free
 }
 {
  bone1Name=dyn_under_jacket_fr_l_MSTR
  bone2Name=spine_01
  swing1LimitDegrees=45
  swing1Motion=Free
  swing2LimitDegrees=45
  swing2Motion=Free
  twistLimitDegrees=45
  twistMotion=Free
 }
 {
  bone1Name=dyn_under_jacket_fr_r_MSTR
  bone2Name=spine_01
  swing1LimitDegrees=45
  swing1Motion=Free
  swing2LimitDegrees=45
  swing2Motion=Free
  twistLimitDegrees=45
  twistMotion=Free
 }
 {
  bone1Name=dyn_thigh_pouch_l
  bone2Name=thigh_l
  swing1LimitDegrees=45
  swing1Motion=Free
  swing2LimitDegrees=45
  swing2Motion=Free
  twistLimitDegrees=45
  twistMotion=Free
 }
 {bone1Name=dyn_jacket_bk_02, bone2Name=dyn_jacket_bk_01, swing1LimitDegrees=5, swing2LimitDegrees=20, twistLimitDegrees=8}
 {
  bone1Name=dyn_jacket_bk_01
  bone2Name=spine_02
  swing1LimitDegrees=45
  swing1Motion=Free
  swing2LimitDegrees=45
  swing2Motion=Free
  twistLimitDegrees=45
  twistMotion=Free
 }
]
