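# Physics asset for the Bone Wasp male medium character body, providing the rigid-body setup for its dynamic necklace.
# Two box bodies form the necklace chain, hung from spine_05, with capsules on spine_04 and spine_05 standing in for the upper torso.
# Limited swing and twist constraints are tighter at the spine attachment than between necklace segments.
# The joint between the two spine bodies leaves all angular motion free.
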PhysicsAsset /BRCosmetics/Characters/Player/Male/Medium/Bodies/M_MED_Bone_Wasp/Meshes/M_MED_Bone_Wasp_Dyn_Necklace
bodies {
 dyn_necklace_01 [{shapeType=Box, extentX=7.5, extentY=1.6, extentZ=6.5, center=<7.462 0 -0.0001>}]
 dyn_necklace_02 [{shapeType=Box, extentX=10, extentY=5.5, extentZ=6.5, center=<5.3735 -1.4113 0>}]
 spine_04 [{shapeType=Capsule, length=18.75, radius=9.25, center=<-0.3285 6.6193 0>}]
 spine_05 [{shapeType=Capsule, length=17, radius=11, center=<-6.5262 4.62 0>}]
}
constraints [
 {bone1Name=dyn_necklace_02, bone2Name=dyn_necklace_01, swing1LimitDegrees=45, swing2LimitDegrees=45, twistLimitDegrees=10}
 {bone1Name=dyn_necklace_01, bone2Name=spine_05, swing1LimitDegrees=30, swing2LimitDegrees=30, twistLimitDegrees=5}
 {
  bone1Name=spine_05
  bone2Name=spine_04
  swing1LimitDegrees=45
  swing1Motion=Free
  swing2LimitDegrees=45
  swing2Motion=Free
  twistLimitDegrees=45
  twistMotion=Free
 }
]
