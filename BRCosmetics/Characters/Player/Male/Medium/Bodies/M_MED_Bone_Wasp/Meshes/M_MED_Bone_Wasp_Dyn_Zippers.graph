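# Physics asset for the Bone Wasp male medium player body that defines bodies for dynamic front-left jacket zippers and three red tails at back, left and right.
# The jacket and under-jacket chains run from the pelvis through small capsules with free angular motion to box zipper pulls, which have limited swing and tight twist.
# Box tail bodies hinge directly off the two-capsule pelvis, with one wide swing axis and narrow side swing and twist.
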PhysicsAsset /BRCosmetics/Characters/Player/Male/Medium/Bodies/M_MED_Bone_Wasp/Meshes/M_MED_Bone_Wasp_Dyn_Zippers
bodies {
 dyn_jacket_fr_l_01 [{shapeType=Capsule, length=1, radius=1}]
 dyn_jacket_fr_l_02 [{shapeType=Capsule, length=1, radius=1}]
 dyn_jacket_fr_l_zipper [{shapeType=Box, extentX=2.5, extentY=1, extentZ=1, center=<1.311 0 0.0003>}]
 dyn_jacket_fr_l_zipper_MSTR [{shapeType=Capsule, length=0.505, radius=0.505, rotation=<rot 0 -90 0>}]
 dyn_red_tail_bk_01 [{shapeType=Box, extentX=7.5, extentY=7.5, extentZ=1.25, center=<4.307 0 0>}]
 dyn_red_tail_l_01 [{shapeType=Box, extentX=7.5, extentY=7.5, extentZ=2, center=<4.307 0 0>}]
 dyn_red_tail_r_01 [{shapeType=Box, extentX=7.5, extentY=7.5, extentZ=2, center=<4.307 0 0>}]
 dyn_under_jacket_fr_l_01 [{shapeType=Capsule, length=1, radius=1}]
 dyn_under_jacket_fr_l_MSTR [{shapeType=Capsule, length=1, radius=1}]
 dyn_under_jacket_fr_l_zipper [{shapeType=Box, extentX=3.5, extentY=2.5, extentZ=0.85, center=<1.5514 0 0>}]
 dyn_under_jacket_fr_l_zipper_MSTR [{shapeType=Capsule, length=0.505, radius=0.505, rotation=<rot 0 -90 0>}]
 pelvis [
  {shapeType=Capsule, length=6.7188, radius=7.625, center=<-5.7161 -5.7276 0>}
  {shapeType=Capsule, length=9.4688, radius=9, center=<-6.6976 3.9057 0>}
 ]
}
constraints [
 {bone1Name=dyn_jacket_fr_l_zipper, bone2Name=dyn_jacket_fr_l_zipper_MSTR, swing1LimitDegrees=30, swing2LimitDegrees=30, twistLimitDegrees=2}
 {
  bone1Name=dyn_under_jacket_fr_l_zipper
  bone2Name=dyn_under_jacket_fr_l_zipper_MSTR
  swing1LimitDegrees=30
  swing2LimitDegrees=30
  twistLimitDegrees=2
 }
 {
  bone1Name=dyn_jacket_fr_l_zipper_MSTR
  bone2Name=dyn_jacket_fr_l_02
  swing1LimitDegrees=45
  swing1Motion=Free
  swing2LimitDegrees=45
  swing2Motion=Free
  twistLimitDegrees=45
  twistMotion=Free
 }
 {
  bone1Name=dyn_under_jacket_fr_l_zipper_MSTR
  bone2Name=dyn_under_jacket_fr_l_01
  swing1LimitDegrees=45
  swing1Motion=Free
  swing2LimitDegrees=45
  swing2Motion=Free
  twistLimitDegrees=45
  twistMotion=Free
 }
 {
  bone1Name=dyn_under_jacket_fr_l_01
  bone2Name=dyn_under_jacket_fr_l_MSTR
  swing1LimitDegrees=45
  swing1Motion=Free
  swing2LimitDegrees=45
  swing2Motion=Free
  twistLimitDegrees=45
  twistMotion=Free
 }
 {
  bone1Name=dyn_under_jacket_fr_l_MSTR
  bone2Name=pelvis
  swing1LimitDegrees=45
  swing1Motion=Free
  swing2LimitDegrees=45
  swing2Motion=Free
  twistLimitDegrees=45
  twistMotion=Free
 }
 {
  bone1Name=dyn_jacket_fr_l_02
  bone2Name=dyn_jacket_fr_l_01
  swing1LimitDegrees=45
  swing1Motion=Free
  swing2LimitDegrees=45
  swing2Motion=Free
  twistLimitDegrees=45
  twistMotion=Free
 }
 {
  bone1Name=dyn_jacket_fr_l_01
  bone2Name=pelvis
  swing1LimitDegrees=45
  swing1Motion=Free
  swing2LimitDegrees=45
  swing2Motion=Free
  twistLimitDegrees=45
  twistMotion=Free
 }
 {bone1Name=dyn_red_tail_bk_01, bone2Name=pelvis, swing1LimitDegrees=15, swing2LimitDegrees=80, twistLimitDegrees=5}
 {bone1Name=dyn_red_tail_l_01, bone2Name=pelvis, swing1LimitDegrees=15, swing2LimitDegrees=80, twistLimitDegrees=5}
 {bone1Name=dyn_red_tail_r_01, bone2Name=pelvis, swing1LimitDegrees=15, swing2LimitDegrees=80, twistLimitDegrees=5}
]
