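# Post-process style Anim Blueprint for the Bone Wasp outfit, adding secondary motion to jacket, collar, necklace and zippers over the parent mesh's pose.
# Arm-raise collar correctives and bone constraints feed collar and necklace RigidBody sims, cached for two jacket sims selected by backbling presence.
# Zipper sim and skydiving collar offsets finish the pose.
# Jacket gravity varies with swim sprint, skydiving and fall speed; Head Up and crouch values go unused.
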AnimBlueprint /BRCosmetics/Characters/Player/Male/Medium/Bodies/M_MED_Bone_Wasp/Meshes/M_MED_Bone_Wasp_AnimBP
target_skeleton=&/M_MED_Bone_Wasp_Skeleton
graphs [
 AnimGraph {nodes [
   n1 = Break Fort Anim Input Skydiving External Force(FortAnimInput_SkydivingExternalForce=Get SkydivingExternalForceInput()) -> (
    bUseSkydivingVectorForce=~
    SkydivingVectorForce=~
    HeadToPelvisDirection=~
    FloatingMultiplier=~
    FloatingAdditive=~
    DivingMultiplier=~
    DivingAdditive=~
    ParachutingMultiplier=~
    ParachutingAdditive=~
   )
   n2 = Get gravity_override() -> (gravity_override:Vector)
   n3 = RigidBody!(
    ComponentPose=Local To Component(LocalPose=Use cached pose 'preJacketCache'())
    Alpha
    OverrideWorldGravity:Vector=@n2
    ExternalForce=@n1.SkydivingVectorForce
   ) -> (Pose) {
    bClampLinearTranslationLimitToRefPose=true
    bOverrideWorldGravity=true
    overridePhysicsAsset=&/M_MED_Bone_Wasp_Dyn_Jacket
    simulationSpace=BaseBoneSpace
    alphaBoolBlend {blendOption=Linear}
    baseBoneRef {boneName=spine_02}
    componentAppliedLinearAccClamp=<3000 3000 3000>
    componentLinearAccScale=<0.5 0.3 0.3>
    componentLinearVelScale=<0.5 0.3 0.3>
    overrideWorldGravity {z=-980}
   }
   n4 = Component To Local(ComponentPose=@n3)
   n5 = RigidBody(
    ComponentPose=Local To Component(LocalPose=Use cached pose 'preJacketCache'())
    OverrideWorldGravity=@n2
    ExternalForce=@n1.SkydivingVectorForce
   ) {
    bClampLinearTranslationLimitToRefPose=true
    bOverrideWorldGravity=true
    overridePhysicsAsset=&/M_MED_Bone_Wasp_Dyn_Jacket_no_back
    simulationSpace=BaseBoneSpace
    alphaBoolBlend {blendOption=Linear}
    baseBoneRef {boneName=spine_02}
    componentAppliedLinearAccClamp=<3000 3000 3000>
    componentLinearAccScale=<0.5 0.3 0.3>
    componentLinearVelScale=<0.5 0.3 0.3>
    overrideWorldGravity {z=-980}
   }
   n6 = Component To Local(ComponentPose=@n5)
   n7 = Get isWearingBackbling()
   n8 = RigidBody!(
    ComponentPose=Local To Component(LocalPose=Blend Poses by bool(BlendPose_0=@n6, BlendPose_1=@n4, bActiveValue=@n7) {blendType=Linear})
    Alpha
    ExternalForce
   ) -> (Pose) {
    bClampLinearTranslationLimitToRefPose=true
    bOverrideWorldGravity=true
    overridePhysicsAsset=&/M_MED_Bone_Wasp_Dyn_Zippers
    simulationSpace=BaseBoneSpace
    alphaBoolBlend {blendOption=Linear}
    baseBoneRef {boneName=spine_02}
    componentAppliedLinearAccClamp=<3000 3000 3000>
    componentLinearAccScale=<0.5 0.3 0.3>
    componentLinearVelScale=<0.5 0.3 0.3>
    overrideWorldGravity {z=-980}
   }
   n9 = Get isSkyDiving() -> (isSkyDiving:double)
   n10 = "Transform (Modify) Bone - Bone: dyn_collar_back_MSTR"(ComponentPose=@n8, Alpha=@n9, Translation="-4,0,2", Rotation="-13,0,0", Scale="1,1,1") {
    rotationMode=BMM_Additive
    rotationSpace=BCS_BoneSpace
    translationMode=BMM_Additive
    translationSpace=BCS_BoneSpace
    alphaBoolBlend {blendOption=Linear}
    alphaScaleBiasClamp {bInterpResult=true}
    boneToModify {boneName=dyn_collar_back_MSTR}
    rotation {pitch=-13}
    translation {x=-4, z=2}
   }
   n11 = "Transform (Modify) Bone - Bone: dyn_collar_bk_l_MSTR"(ComponentPose=@n10, Alpha=@n9, Translation="-4,0,2", Rotation="-10,0,0", Scale="1,1,1") {
    rotationMode=BMM_Additive
    rotationSpace=BCS_BoneSpace
    translationMode=BMM_Additive
    translationSpace=BCS_BoneSpace
    alphaBoolBlend {blendOption=Linear}
    alphaScaleBiasClamp {bInterpResult=true}
    boneToModify {boneName=dyn_collar_bk_l_MSTR}
    rotation {pitch=-10}
    translation {x=-4, z=2}
   }
   n12 = "Transform (Modify) Bone - Bone: dyn_collar_bk_r_MSTR"(ComponentPose=@n11, Alpha=@n9, Translation="-4,0,2", Rotation="-10,0,0", Scale="1,1,1") {
    rotationMode=BMM_Additive
    rotationSpace=BCS_BoneSpace
    translationMode=BMM_Additive
    translationSpace=BCS_BoneSpace
    alphaBoolBlend {blendOption=Linear}
    alphaScaleBiasClamp {bInterpResult=true}
    boneToModify {boneName=dyn_collar_bk_r_MSTR}
    rotation {pitch=-10}
    translation {x=-4, z=2}
   }
   Output Pose(Result=Component To Local(ComponentPose=@n12)) {layerGroup=None}
   n13 = Get arm_up_r() -> (arm_up_r:double)
   n14 = "Transform (Modify) Bone - Bone: dyn_collar_fr_r_MSTR"(
    ComponentPose=Local To Component(LocalPose=Copy Pose From Mesh() {bUseAttachedParent=true})
    Alpha=@n13
    Translation="-1.453,-0.000,-2.807"
    Rotation="-50.000,0.000,0.000"
    Scale="1,1,1"
   ) {
    rotationMode=BMM_Additive
    rotationSpace=BCS_BoneSpace
    translationMode=BMM_Additive
    translationSpace=BCS_BoneSpace
    alphaBoolBlend {blendOption=Linear}
    alphaScaleBiasClamp {bMapRange=true, inRange {max=0.75, min=0.5}}
    boneToModify {boneName=dyn_collar_fr_r_MSTR}
    rotation {pitch=-49.9999}
    translation {x=-1.453, z=-2.807}
   }
   n15 = "Transform (Modify) Bone - Bone: dyn_collar_r_MSTR"(ComponentPose=@n14, Alpha=@n13, Translation="-1.551,0.000,0.341", Rotation="0,0,0", Scale="1,1,1") {
    rotationMode=BMM_Additive
    rotationSpace=BCS_BoneSpace
    translationMode=BMM_Additive
    translationSpace=BCS_BoneSpace
    alphaBoolBlend {blendOption=Linear}
    alphaScaleBiasClamp {bMapRange=true, inRange {max=0.75, min=0.5}}
    boneToModify {boneName=dyn_collar_r_MSTR}
    translation {x=-1.551, z=0.341}
   }
   n16 = "Transform (Modify) Bone - Bone: dyn_collar_fr_l_MSTR"(
    ComponentPose=@n15
    Alpha=Get arm_up_l() -> (arm_up_l:double)
    Translation="-0.385,-0.000,-1.420"
    Rotation="-15.000,0.000,0.000"
    Scale="1,1,1"
   ) {
    rotationMode=BMM_Additive
    rotationSpace=BCS_BoneSpace
    translationMode=BMM_Additive
    translationSpace=BCS_BoneSpace
    alphaBoolBlend {blendOption=Linear}
    alphaScaleBiasClamp {interpSpeedDecreasing=1, interpSpeedIncreasing=1}
    boneToModify {boneName=dyn_collar_fr_l_MSTR}
    rotation {pitch=-15}
    translation {x=-0.385, z=-1.42}
   }
   n17 = "Constraint - dyn_under_jacket_fr_l_zipper_MSTR (PPPP)"!(
    ComponentPose=@n16
    Alpha=1
    ConstraintWeights_0=0
    ConstraintWeights_1:float=0.6522
    ConstraintWeights_2:float=0
    ConstraintWeights_3:float=0.3478
   ) -> (Pose) {
    alphaBoolBlend {blendOption=Linear}
    boneToModify {boneName=dyn_under_jacket_fr_l_zipper_MSTR}
    constraintSetup [
     {targetBone {boneName=spine_01}}
     {targetBone {boneName=dyn_under_jacket_fr_l_01}}
     {targetBone {boneName=thigh_twist_01_l}}
     {targetBone {boneName=pelvis}}
    ]
    constraintWeights [1,1,1,1]
   }
   n18 = "Constraint - dyn_jacket_fr_l_zipper_MSTR (P)"!(ComponentPose=@n17, Alpha=1, ConstraintWeights_0=1, ConstraintWeights_1:float=0) -> (Pose) {
    alphaBoolBlend {blendOption=Linear}
    boneToModify {boneName=dyn_jacket_fr_l_zipper_MSTR}
    constraintSetup [{targetBone {boneName=dyn_jacket_fr_l_02}}]
    constraintWeights [1,0]
   }
   n19 = "Constraint - dyn_under_jacket_fr_r_MSTR (PPPPPP)"!(
    ComponentPose=@n18
    Alpha=1
    ConstraintWeights_0=0.0282
    ConstraintWeights_1:float=0.078
    ConstraintWeights_2:float=0
    ConstraintWeights_3:float=0
    ConstraintWeights_4:float=0.8937
    ConstraintWeights_5:float=0.0002
   ) -> (Pose) {
    alphaBoolBlend {blendOption=Linear}
    boneToModify {boneName=dyn_under_jacket_fr_r_MSTR}
    constraintSetup [
     {targetBone {boneName=thigh_twist_01_r}}
     {targetBone {boneName=spine_01}}
     {targetBone {boneName=spine_02}}
     {targetBone {boneName=spine_03}}
     {targetBone {boneName=pelvis}}
     {targetBone {boneName=thigh_twist_01_l}}
    ]
    constraintWeights [1,1,1,1,1,1]
   }
   n20 = "Constraint - dyn_under_jacket_fr_l_MSTR (PPPPPP)"!(
    ComponentPose=@n19
    Alpha=1
    ConstraintWeights_0=0.0002
    ConstraintWeights_1:float=0.078
    ConstraintWeights_2:float=0
    ConstraintWeights_3:float=0
    ConstraintWeights_4:float=0.8937
    ConstraintWeights_5:float=0.0282
   ) -> (Pose) {
    alphaBoolBlend {blendOption=Linear}
    boneToModify {boneName=dyn_under_jacket_fr_l_MSTR}
    constraintSetup [
     {targetBone {boneName=thigh_twist_01_r}}
     {targetBone {boneName=spine_01}}
     {targetBone {boneName=spine_02}}
     {targetBone {boneName=spine_03}}
     {targetBone {boneName=pelvis}}
     {targetBone {boneName=thigh_twist_01_l}}
    ]
    constraintWeights [1,1,1,1,1,1]
   }
   n21 = RigidBody(ComponentPose=@n20) {
    bClampLinearTranslationLimitToRefPose=true
    bOverrideWorldGravity=true
    overridePhysicsAsset=&/M_MED_Bone_Wasp_Dyn_Collar
    simulationSpace=BaseBoneSpace
    alphaBoolBlend {blendOption=Linear}
    baseBoneRef {boneName=spine_01}
    componentAppliedLinearAccClamp=<3000 3000 3000>
    componentLinearAccScale=<0.5 0.4 0.4>
    componentLinearVelScale=<0.5 0.4 0.4>
    overrideWorldGravity {z=-980}
   }
   n22 = RigidBody(ComponentPose=@n21, ExternalForce=@n1.SkydivingVectorForce) {
    bClampLinearTranslationLimitToRefPose=true
    bOverrideWorldGravity=true
    overridePhysicsAsset=&/M_MED_Bone_Wasp_Dyn_Necklace
    simulationSpace=BaseBoneSpace
    alphaBoolBlend {blendOption=Linear}
    baseBoneRef {boneName=spine_01}
    componentAppliedLinearAccClamp=<3000 3000 3000>
    componentLinearAccScale=<0.5 0.3 0.3>
    componentLinearVelScale=<0.5 0.3 0.3>
    overrideWorldGravity {z=-980}
   }
   Save cached pose 'preJacketCache'(Pose=Component To Local(ComponentPose=@n22)) {cacheName=preJacketCache}
  ]}
 EventGraph {nodes [
   Event BlueprintUpdateAnimation()
   > n1 = Sequence!(execute) -> (then_0, then_1, then_2:Exec, then_3:Exec)
   n2 = BreakTransform(InTransform=GetDeltaTransformFromRefPose(self=GetOwnerSkeletalMeshComponent(), BoneName=upperarm_r, BaseName=spine_05))
   > Set arm_up_r(arm_up_r:double=MapRangeClamped(Value=BreakRotator(InRot=@n2.Rotation).Yaw, InRangeA=40, InRangeB=120)) -> (Output_Get:double)
   n3 = BreakTransform(InTransform=GetDeltaTransformFromRefPose(self=GetOwnerSkeletalMeshComponent(), BoneName=upperarm_l, BaseName=spine_05))
   > Set arm_up_l(arm_up_l:double=MapRangeClamped(Value=BreakRotator(InRot=@n3.Rotation).Yaw, InRangeA=50, InRangeB=105)) -> (Output_Get:double)
   n4 = BreakTransform(InTransform=GetDeltaTransformFromRefPose(self=GetOwnerSkeletalMeshComponent(), BoneName=head, BaseName=spine_05))
   > Set Head Up(Head Up:double=MapRangeClamped(Value=BreakRotator(InRot=@n4.Rotation).Pitch, InRangeA=0, InRangeB=15)) -> (Output_Get:double)
   n5 = Branch(execute=@n1.then_1, Condition=Get bIsSwimSprinting() -> (bIsSwimSprinting:bool))
   > Set gravity_override(gravity_override="0, 0, 0")
   n6 = Branch(execute=@n5.else, Condition=Get bIsSkyDiving())
   > n7 = Branch(Condition=Get bIsParachuteOpened())
   Set gravity_override(execute=[@n6.else, @n7.then], gravity_override="0, 0,-980")
   n8 = Branch(execute=@n7.else, Condition="float > float"(A=Get PawnVelocityZ(), B=50) {is=title K2Node_CallFunction})
   > Set gravity_override(gravity_override="0, 0,200")
   Set gravity_override(execute=@n8.else, gravity_override="0, 0,100")
   n9 = Get bIsSkyDiving()
   n10 = Get bIsSkydiveDiveMode()
   n11 = Get bIsSkydiveFloating()
   Set isSkyDiving(execute=@n1.then_2, isSkyDiving:double="To Float (Boolean)"(InBool=OR Boolean!(A=@n9, B=@n10, C:bool=@n11) -> (ReturnValue))) -> (Output_Get:double)
   Event OnCharacterCustomizationLoaded()
   > Do Once()
   n12 = CheckForBackbling()
   > Set isWearingBackbling(isWearingBackbling=OR Boolean!(A=@n12, B=@n12.bIsBackblingLong) -> (ReturnValue))
   Event OnEndCrouching()
   > Set crouching alpha(crouching alpha:double=0) -> (Output_Get:double)
   Event OnBeginCrouching()
   > Set crouching alpha(crouching alpha=1)
   TryGetPawnOwner()
   TryGetPawnOwner()
  ]}
]
variables [
 arm_up_l {type=number}
 arm_up_r {type=number}
 crouching alpha {type=number}
 gravity_override {type=Vector}
 Head Up {type=number}
 isSkyDiving {type=number}
 isWearingBackbling {type=boolean}
]
